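; Gameplay ability that plays speech-bubble barks for a specialist NPC when it is damaged, eliminated, hired or uses its specialist ability.
; On activation it binds to the owning pawn's damage, death and convert-component events and waits for the specialist gameplay event.
; Each trigger passes its bark tag to a shared helper, with a short random delay for specialist barks.
; A flag and timer throttle damage barks; ability end unbinds the delegates.
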(event Ability|EventActivateAbility
  (bind _returnvalue (Ability|GetActivatingPawn))
  ;[0 0 0] Activate Ability
  (|CallParentFunction)
  (Utilities|IsValid _returnvalue
    (:Is%20Valid
      (Variables|OwnerInfo|SetOwningPawn _returnvalue)
      (CallFunction|StartListeningtoDamageandElim)
      (CallFunction|StartListeningtoSpecialistAbility)
      (CallFunction|StartListeningtoHire))
    (:Is%20Not%20Valid
      (Ability|EndAbility))))

(event Custom|StartListeningToDamageAndElim
  ;[0 0 0] Start listening to bark events
  (EventDispatchers|BindEventtoOnDamaged (Variables|OwnerInfo|GetOwningPawn) (EventDispatchers|CreateEvent :!event Custom|OnPawnDamaged))
  (EventDispatchers|BindEventtoOnDied (Variables|OwnerInfo|GetOwningPawn) (EventDispatchers|CreateEvent :!event Custom|OnPawnDied)))

(event Custom|StartListeningToSpecialistAbility
  ;[0 0 0] Start listening to bark events
  (Ability|Async|WaitGameplayEventToActor (Variables|OwnerInfo|GetOwningPawn) (Variables|InstanceParameters|GetSpecialistEvent)
    (:EventReceived
      (CallFunction|PlayRandomBarkforTag (Variables|Barks|GetSpecialistBarkTag) true))))

(event Custom|StartListeningToHire
  ;[0 0 0] Start listening to bark events
  (bind _asfort_player_pawn (Utilities|Casting|CastToFortPlayerPawn (Variables|OwnerInfo|GetOwningPawn)))
  (bind _output_get (Variables|Default|SetConvert%20Component (Variables|/Script/FortniteGame.FortPlayerPawn|GetConvertComponent _asfort_player_pawn)))
  (ConvertEvents|BindEventtoOnConvertedEvent _output_get (EventDispatchers|CreateEvent :!event Custom|OnPawnHired)))

(event Custom|OnPawnDamaged
  (params (DamagedActor :type Actor%20Object%20Reference) (Damage :type Float%20%28single-precision%29) (InstigatedBy :type Controller%20Object%20Reference) (DamageCauser :type Actor%20Object%20Reference) (HitLocation :type Vector) (FHitComponent :type Primitive%20Component%20Object%20Reference) (BoneName :type Name) (Momentum :type Vector))
  ;[0 0 0] NPC damaged, eliminated or hired
  (if (and (> Damage 1.0) (not (Variables|OwnerInfo|GetbRecentlyDamaged)))
    (CallFunction|PlayRandomBarkforTag (Variables|Barks|GetPawnDamagedBarkTag))
    (Variables|OwnerInfo|SetbRecentlyDamaged true)
    (Utilities|FlowControl|Delay (Variables|Default|GetMinTimeToRandomBark))
    (Variables|OwnerInfo|SetbRecentlyDamaged false)))

(event Custom|OnPawnDied
  (params (DamagedActor :type Actor%20Object%20Reference) (Damage :type Float%20%28single-precision%29) (InstigatedBy :type Controller%20Object%20Reference) (DamageCauser :type Actor%20Object%20Reference) (HitLocation :type Vector) (FHitComponent :type Primitive%20Component%20Object%20Reference) (BoneName :type Name) (Momentum :type Vector))
  ;[0 0 0] NPC damaged, eliminated or hired
  (CallFunction|PlayRandomBarkforTag (Variables|Barks|GetPawnElimBarkTag)))

(event Custom|OnPawnHired
  (params (InstigatorPawn :type Fort%20Pawn%20Object%20Reference) (ConvertedPawn :type Fort%20Pawn%20Object%20Reference))
  ;[0 0 0] NPC damaged, eliminated or hired
  (CallFunction|PlayRandomBarkforTag (Variables|Barks|GetHireBarkTag)))

(event Ability|EventOnEndAbility
  (params (bWasCancelled :type Boolean))
  ;[0 0 0] On end ability - unbind events
  (|CallParentFunction bWasCancelled)
  (EventDispatchers|UnbindEventfromOnDamaged (Variables|OwnerInfo|GetOwningPawn) (EventDispatchers|CreateEvent :!event Custom|OnPawnDamaged))
  (EventDispatchers|UnbindEventfromOnDied (Variables|OwnerInfo|GetOwningPawn) (EventDispatchers|CreateEvent :!event Custom|OnPawnDied))
  (ConvertEvents|UnbindEventfromOnConvertedEvent (Variables|Default|GetConvert%20Component) (EventDispatchers|CreateEvent :!event Custom|OnPawnHired)))

(event Custom|PlayRandomBarkForTag
  (params (BarkTag :type Gameplay%20Tag%20Structure) (bDelay :type Boolean))
  ;[0 0 0] Helper events
  (if bDelay
    (Utilities|FlowControl|Delay (Math|Random|RandomFloatinRange 0.5 1.0))
    (else
      _))
  (Class|BPFLAICommonFunctions|PlaySpeechBubble (Variables|OwnerInfo|GetOwningPawn) BarkTag))

(event Ability|EventFailedToActivatePassiveAbility
  (params (ActorInfo :type Gameplay%20Ability%20Actor%20Info%20Structure))
  ;[0 0 0] Helper events
  (|CallParentFunction ActorInfo))

(fn IsOwnerInSafeZone
  (bind _asfort_player_pawn (Utilities|Casting|CastToFortPlayerPawn (Variables|OwnerInfo|GetOwningPawn)))
  (return (Variables|/Script/FortniteGame.FortPlayerPawn|GetbIsInsideSafeZone _asfort_player_pawn)))
